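# Surface material for a Music Sequencer device pad, showing a tinted icon and a parameter-driven glow.
# Base color adds a faint tiled grid to a color, then blends to a second color through the alpha of a center-scaled icon texture.
# A tiled cloth normal map is scaled by a strength vector.
# The glow, black by default, uses a rounded-square SphereMask from powered, centered UVs and adds squared glow color inside the icon.
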Material /Game/Creative/Devices/Music_Sequencer/M_Music_Seq_Pad
graph {
 properties {pixelDepthOffsetMode=PDOM_Legacy, previewMesh=&/Audio_Trap_Floor_Button, usedShadingModels=Default Lit}
 nodes [
  n1 = TextureSample(UVs=TextureCoordinate() {uTiling=4, vTiling=4}) {texture=/Game/Weapons/Prototype/ObjectMover/Textures/T_Movetool_Transition}
  n2 = ScalarParameter() {defaultValue=0.02, parameterName=GridScale}
  n3 = TextureSampleParameter2D(UVs=MaterialFunctionCall(Texture Scale=ScalarParameter() {defaultValue=0.175, parameterName=ScaleUVs}) {materialFunction=&ScaleUVsByCenter}.UVs) {parameterName=Texture, texture=/Game/UI/Foundation/Textures/Icons/Replay/T-Icon-Replay-Pause-L}
  LinearInterpolate(
   A=Add(A=Multiply(A=@n1.R, B=@n2), B=VectorParameter() {parameterName=BaseColor, defaultValue=<rgba 0.0321 0.1374 0.6858 0>}.RGB)
   B=VectorParameter() {parameterName=TextureColor, defaultValue=<rgba 0.1665 0.5931 0.9497 0>}.RGB
   Alpha=@n3.A
  )
  Multiply(
   A=TextureSample(UVs=TextureCoordinate() {uTiling=6, vTiling=6}) {samplerType=SAMPLERTYPE_Normal, texture=/Game/Global/Textures/Default/Noise/T_FN_FluffyCloth_N}.RGB
   B=VectorParameter() {parameterName=NormalStrength, defaultValue=<rgba 0.05 0.05 1 0>}.RGB
  )
  n4 = Abs(None=ConstantBiasScale(None=TextureCoordinate()) {bias=-0.5, scale=2})
  n5 = ScalarParameter() {defaultValue=6, parameterName=Exp}
  n6 = VectorParameter() {parameterName=GlowColor, defaultValue=<rgba 0 0 0 10>}
  n7 = Multiply(A=@n6.RGB, B=@n6.RGB)
  Multiply(
   A=SphereMask(
    A=Power(Base=@n4, Exp=@n5)
    B=Constant()
    Radius=ScalarParameter() {defaultValue=0.5, parameterName=Radius}
    Hardness=ScalarParameter() {parameterName=Hardness}
   )
   B=Add(A=@n6.RGB, B=Multiply(A=@n3.A, B=@n7))
  )
 ]
}
